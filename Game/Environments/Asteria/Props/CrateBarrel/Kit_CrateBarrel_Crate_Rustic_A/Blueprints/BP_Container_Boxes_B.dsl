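; BP_Container_Boxes_B is a rustic crate container prop that matches its instanced secondary mesh to the building's texture data.
; The construction script runs the parent's logic, then continues only if building texture data exists.
; It puts the dynamic material instance from the main static mesh component on the instanced mesh's first slot, then colors that mesh with the first entry's tint.
; BeginPlay, overlap and Tick events are empty.
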(event EventBeginPlay)

(event Collision|EventActorBeginOverlap
  (params (OtherActor :type Actor%20Object%20Reference)))

(event EventTick
  (params (DeltaSeconds :type Float%20%28single-precision%29)))

(fn ConstructionScript
  (|CallParentFunction)
  (bind _outtexturedataarray (Building|GetBuildingTextureData))
  (if (> (Utilities|Array|Length _outtexturedataarray) 0)
    (Utilities|FlowControl|Sequence
      (:then_0
        (Rendering|Material|SetMaterial (Variables|Default|GetInstancedStaticMesh) 0 (Class|BTechArtGlobalLibrary|GetDynamicMaterialInstance (Variables|Debug|GetStaticMeshComponent))))
      (:then_1
        (Default|SetColoronSecondaryMesh (Math|Conversions|ToLinearColor(Color) (Variables|/Script/FortniteGame.BuildingTextureData|GetTintColor (Utilities|Array|Get(acopy) _outtexturedataarray))) (Variables|Default|GetInstancedStaticMesh))))))
